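# Physics asset for the Storm King enemy's skeletal mesh, providing very large collision bodies for pelvis, upper spine, neck, head, horns, arms, hands and one whip bone.
# Capsules cover torso, head, arms and whip_06, boxes cover the hands, and each curved horn is built from four named boxes.
# No leg bodies or constraints are defined, so the bodies are not jointed as a ragdoll.
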PhysicsAsset /Game/Characters/Enemies/StormKing/Meshes/SK_StormKing_01_Physics
bodies {
 hand_l [{shapeType=Box, extentX=573, extentY=256, extentZ=307, center=<312.6117 -60.4404 0.0734>, rotation=<rot 5.4524 -6.3465 -0.6054>}]
 hand_r [{shapeType=Box, extentX=573, extentY=256, extentZ=307, center=<-312.6117 60.4404 0.0734>, rotation=<rot 5.4524 -6.3465 -0.6054>}]
 head [{shapeType=Capsule, length=87.2235, radius=223.8244, center=<13.3354 73.1132 -0.025>, rotation=<rot 90 0 120.0093>}]
 horn_l [
  {
   shapeType=Box
   extentX=800
   extentY=120
   extentZ=30
   shapeName=HornTipL
   center=<15.1627 815.7276 277.432>
   rotation=<rot 20.029 86.5694 -0.9255>
  }
  {
   shapeType=Box
   extentX=450
   extentY=160
   extentZ=50
   shapeName=HornCurve1L
   center=<203.7987 423.6567 90.7609>
   rotation=<rot 21.3824 144.6635 16.3809>
  }
  {
   shapeType=Box
   extentX=200
   extentY=250
   extentZ=100
   shapeName=HornCurve2L
   center=<296.5105 263.124 14.0894>
   rotation=<rot 19.9235 87.9441 -0.6704>
  }
  {
   shapeType=Box
   extentX=500
   extentY=150
   extentZ=80
   shapeName=HornCurve3L
   center=<131.8191 52.1953 -12.6503>
   rotation=<rot -0.818 28.246 -18.1489>
  }
 ]
 horn_r [
  {
   shapeType=Box
   extentX=800
   extentY=120
   extentZ=30
   shapeName=HornTipR
   center=<-11.6124 815.0092 277.0532>
   rotation=<rot 19.9944 93.7306 1.5277>
  }
  {
   shapeType=Box
   extentX=450
   extentY=160
   extentZ=50
   shapeName=HornCurve1R
   center=<-201.3808 422.3866 91.6782>
   rotation=<rot 21.295 35.18 -16.6408>
  }
  {
   shapeType=Box
   extentX=200
   extentY=250
   extentZ=100
   shapeName=HornCurve2R
   center=<-294.3266 261.457 14.5053>
   rotation=<rot 19.9198 92.1921 0.7776>
  }
  {
   shapeType=Box
   extentX=500
   extentY=150
   extentZ=80
   shapeName=HornCurve3R
   center=<-128.7787 51.4715 -12.2418>
   rotation=<rot 0.9487 -28.2053 -17.8119>
  }
 ]
 lowerarm_l [{shapeType=Capsule, length=600, radius=125, center=<345.4429 0.0054 0.0007>, rotation=<rot 90 0 0>}]
 lowerarm_r [{shapeType=Capsule, length=600, radius=125, center=<-410.2976 0.0016 0>, rotation=<rot 90 0 0>}]
 neck_02 [{shapeType=Capsule, length=250, radius=150, center=<-0.0001 -31.8003 0.0001>}]
 pelvis [{shapeType=Capsule, length=227.9881, radius=230.6559, rotation=<rot 90 0 0>}]
 spine_04 [{shapeType=Capsule, length=133.5, radius=261.25, center=<0 41.1314 5.6163>}]
 spine_05 [
  {shapeType=Capsule, length=250, radius=373.1843, shapeName=Pecs, center=<152.3871 -51.8957 0>}
  {shapeType=Capsule, length=450, radius=250, center=<403.7046 -11.751 -0.0036>}
 ]
 upperarm_l [
  {shapeType=Capsule, length=300, radius=150, center=<20.2703 17.9567 118.7755>, rotation=<rot 61.7523 0 4.6284>}
  {shapeType=Capsule, length=743.75, radius=150, center=<346.7809 -0.0047 0>, rotation=<rot 90 0 0>}
 ]
 upperarm_r [
  {shapeType=Capsule, length=300, radius=150, center=<20.2703 17.9567 -118.7755>, rotation=<rot 61.7523 0 4.6284>}
  {shapeType=Capsule, length=743.75, radius=150, center=<-355.8177 -0.0473 0.0122>, rotation=<rot 90 0 0>}
 ]
 whip_06 [{shapeType=Capsule, length=120.9988, radius=174.2145, center=<-71.3086 0.0002 0>, rotation=<rot 90 0 0>}]
}
constraints []
